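# Translucent, unlit, two-sided Niagara sprite and mesh particle material for a glowing pink edge band effect on the female Obsidian pickaxe.
# Two oppositely panning soft-noise samples are subtracted to blend between minimum and maximum glow, then tinted and passed through MF_FX_EmissiveAdjustments.
# An inverted DepthFade confines the band to geometry intersections, with a constant 1 on mobile.
# A panning normal-noise texture on UV1, scaled by intensity, supplies world position offset wobble.
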Material /BRCosmetics/Weapons/FORT_Melee/Pickaxe_Obsidian_Female/Effects/M_FX_Obsidian_EdgeDecal
graph {
 properties {
  bUsedWithNiagaraMeshParticles=true
  bUsedWithNiagaraSprites=true
  blendMode=BLEND_Translucent
  pixelDepthOffsetMode=PDOM_Legacy
  shadingModel=MSM_Unlit
  twoSided=true
  usedShadingModels=Unlit
 }
 nodes [
  n1 = ScalarParameter() {defaultValue=0.8, group=Edge Band Emissive Noise, parameterName=EdgeNoise_Emissive_Min}
  n2 = ScalarParameter() {defaultValue=10, group=Edge Band Emissive Noise, parameterName=EdgeNoise_Emissive_Max}
  n3 = Multiply(
   A=TextureCoordinate()
   B=AppendVector(
    A=ScalarParameter() {defaultValue=5, group=Edge Band Emissive Noise, parameterName=EdgeNoise_Tile_X}
    B=ScalarParameter() {defaultValue=5, group=Edge Band Emissive Noise, parameterName=EdgeNoise_Tile_Y}
   )
  )
  n4 = ScalarParameter() {defaultValue=0.2, group=Edge Band Emissive Noise, parameterName=EdgeNoise_A_Panner_X}
  n5 = ScalarParameter() {defaultValue=0.5, group=Edge Band Emissive Noise, parameterName=EdgeNoise_A_Panner_Y}
  n6 = TextureSampleParameter2D(UVs=Panner(Coordinate=@n3, Speed=AppendVector(A=@n4, B=@n5))) {
   group=Edge Band Emissive Noise
   parameterName=EdgeNoise_Texture
   texture=/Game/Characters/Player/Male/Medium/Bodies/M_MED_Viper/Textures/Glitch/T_SoftNoise
  }
  n7 = ScalarParameter() {defaultValue=-0.1, group=Edge Band Emissive Noise, parameterName=EdgeNoise_B_Panner_X}
  n8 = ScalarParameter() {defaultValue=0.025, group=Edge Band Emissive Noise, parameterName=EdgeNoise_B_Panner_Y}
  n9 = TextureSampleParameter2D(UVs=Panner(Coordinate=@n3, Speed=AppendVector(A=@n7, B=@n8))) {
   group=Edge Band Emissive Noise
   parameterName=EdgeNoise_Texture
   texture=&T_SoftNoise
  }
  n10 = Clamp(None=Subtract(A=@n6.R, B=@n9.R))
  MaterialFunctionCall(Emissive=Multiply(A=LinearInterpolate(A=@n1, B=@n2, Alpha=@n10), B=Constant3Vector() {constant=<rgba 0.9823 0.4911 0.9004 1>})) {materialFunction=&MF_FX_EmissiveAdjustments}
  n11 = Multiply(
   A=TextureCoordinate() {coordinateIndex=1}
   B=AppendVector(A=ScalarParameter() {defaultValue=1, parameterName=WPO-Tiling_X}, B=ScalarParameter() {defaultValue=1, parameterName=WPO-Tiling_Y})
  )
  n12 = AppendVector(A=ScalarParameter() {defaultValue=-0.2, parameterName=WPO-Speed_X}, B=ScalarParameter() {defaultValue=0.05, parameterName=WPO-Speed_Y})
  Multiply(
   A=TextureSampleParameter2D(UVs=Panner(Coordinate=@n11, Speed=@n12)) {parameterName=WPO-Texture, samplerType=SAMPLERTYPE_Normal, texture=/Game/Global/Textures/Default/Noise/T_Normal_Noise01}.RGB
   B=ScalarParameter() {defaultValue=1, parameterName=WPO-Intensity}
  )
  n13 = OneMinus(None=DepthFade(
    Opacity=ScalarParameter() {defaultValue=1, parameterName=Opacity}
    FadeDistance=ScalarParameter() {defaultValue=2, parameterName=FadeDistance}
   ))
  ShadingPathSwitch(Default=@n13, Deferred=@n13, Mobile=Constant() {r=1}) {desc=Migrated from FeatureLevelSwitch}
 ]
}
